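; Cosmetic actor for the "Future" vehicle boost effect, which sets up a Niagara boost when the cosmetic is applied.
; On the applied event it assigns the configured Niagara system asset to its boost component and calls SetupBoostCount.
; It then offsets the boost component per car through the shared vehicle cosmetics function library, using the vehicle cosmetics interface.
; The construction script only runs the parent's logic.
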(event Vehicle|Cosmetics|EventOnCosmeticApplied
  (bind _ns_boost (Variables|Default|GetNS_Boost))
  (Niagara|SetNiagaraSystemAsset _ns_boost (Variables|Default|GetNiagaraSystem) true)
  (CallFunction|SetupBoostCount)
  (Class|BPVehicleCosmeticsFunctions|SetBoostOffsetbyCar (Variables|Vehicle|Cosmetics|GetVehicleCosmeticsInterface) _ns_boost))

(fn ConstructionScript
  (|CallParentFunction))
